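# Physics asset for the StrayManta medium male body cosmetic, apparently for secondary motion of six bags at front, middle and back on both sides.
# Each bag has a tiny base capsule constrained to the pelvis with free rotation, carrying a larger dynamic-bone capsule with narrow swing on one axis, wider on the other, and zero twist.
# Capsules on the pelvis, lower spine and thighs provide lower-body collision shapes.
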PhysicsAsset /BRCosmetics/Characters/Player/Male/Medium/Bodies/M_MED_StrayManta/Meshes/M_MED_StrayManta_Physics
bodies {
 Bace_Bags_Bk_L [{shapeType=Capsule, length=0.505, radius=0.505, shapeName=Bace_Bags_Bk_L_capsule, rotation=<rot 0 -90 0>}]
 Bace_Bags_Bk_R [{shapeType=Capsule, length=0.505, radius=0.505, shapeName=Bace_Bags_Bk_R_capsule, rotation=<rot 0 -90 0>}]
 Bace_Bags_Fr_L [{shapeType=Capsule, length=0.505, radius=0.505, shapeName=Bace_Bags_Fr_L_capsule, rotation=<rot 0 -90 0>}]
 Bace_Bags_Fr_R [{shapeType=Capsule, length=0.505, radius=0.505, shapeName=Bace_Bags_Fr_R_capsule, rotation=<rot 0 -90 0>}]
 Bace_Bags_Mid_L [{shapeType=Capsule, length=0.505, radius=0.505, shapeName=Bace_Bags_Mid_L_capsule, rotation=<rot 0 -90 0>}]
 Bace_Bags_Mid_R [{shapeType=Capsule, length=0.505, radius=0.505, shapeName=Bace_Bags_Mid_R_capsule, rotation=<rot 0 -90 0>}]
 Dyn_Bags_Bk_L [{
   shapeType=Capsule
   length=6.5429
   radius=2.4064
   shapeName=Dyn_Bags_Bk_L_capsule
   center=<3.913 0.1416 1.2916>
   rotation=<rot 0 -89.4642 -85.3577>
  }]
 Dyn_Bags_Bk_R [{
   shapeType=Capsule
   length=6.1989
   radius=2.7505
   shapeName=Dyn_Bags_Bk_R_capsule
   center=<3.9125 -0.1416 1.2917>
   rotation=<rot 0 -90.5358 -85.3577>
  }]
 Dyn_Bags_Fr_L [{
   shapeType=Capsule
   length=6.9601
   radius=2.4018
   shapeName=Dyn_Bags_Fr_L_capsule
   center=<4.2545 -0.0696 1.3507>
   rotation=<rot 0 -90.0357 -86.8554>
  }]
 Dyn_Bags_Fr_R [{
   shapeType=Capsule
   length=6.9601
   radius=2.4018
   shapeName=Dyn_Bags_Fr_R_capsule
   center=<4.2548 0.0695 1.3506>
   rotation=<rot 0 -89.9643 -86.8554>
  }]
 Dyn_Bags_Mid_L [{
   shapeType=Capsule
   length=6.9591
   radius=2.4133
   shapeName=Dyn_Bags_Mid_L_capsule
   center=<4.2378 0.2676 1.1854>
   rotation=<rot 0 -89.2419 -86.4973>
  }]
 Dyn_Bags_Mid_R [{
   shapeType=Capsule
   length=6.9591
   radius=2.4133
   shapeName=Dyn_Bags_Mid_R_capsule
   center=<4.238 -0.2676 1.1854>
   rotation=<rot 0 -90.7581 -86.4973>
  }]
 pelvis [
  {shapeType=Capsule, length=8.1446, radius=8.8655, shapeName=pelvis_capsule}
  {shapeType=Capsule, length=8.2768, radius=6.7666, shapeName=pelvis_capsule, center=<-3.0352 -6.5089 0>}
 ]
 spine_01 [{shapeType=Capsule, length=5.8469, radius=8.8615, shapeName=spine_01_capsule}]
 thigh_l [{shapeType=Capsule, length=23.7654, radius=8.2166, shapeName=thigh_l_capsule, center=<-20.4727 0 0>, rotation=<rot 90 0 0>}]
 thigh_r [{shapeType=Capsule, length=23.2415, radius=8.7405, shapeName=thigh_r_capsule, center=<20.4727 0 0>, rotation=<rot 90 180 0>}]
}
constraints [
 {
  bone1Name=spine_01
  bone2Name=pelvis
  swing1LimitDegrees=45
  swing1Motion=Free
  swing2LimitDegrees=45
  swing2Motion=Free
  twistLimitDegrees=45
  twistMotion=Free
 }
 {
  bone1Name=thigh_l
  bone2Name=pelvis
  swing1LimitDegrees=45
  swing1Motion=Free
  swing2LimitDegrees=45
  swing2Motion=Free
  twistLimitDegrees=45
  twistMotion=Free
 }
 {
  bone1Name=thigh_r
  bone2Name=pelvis
  swing1LimitDegrees=45
  swing1Motion=Free
  swing2LimitDegrees=45
  swing2Motion=Free
  twistLimitDegrees=45
  twistMotion=Free
 }
 {
  bone1Name=Bace_Bags_Bk_L
  bone2Name=pelvis
  swing1LimitDegrees=45
  swing1Motion=Free
  swing2LimitDegrees=45
  swing2Motion=Free
  twistLimitDegrees=45
  twistMotion=Free
 }
 {bone1Name=Dyn_Bags_Bk_L, bone2Name=Bace_Bags_Bk_L, swing1LimitDegrees=5, swing2LimitDegrees=45, twistLimitDegrees=0}
 {
  bone1Name=Bace_Bags_Bk_R
  bone2Name=pelvis
  swing1LimitDegrees=45
  swing1Motion=Free
  swing2LimitDegrees=45
  swing2Motion=Free
  twistLimitDegrees=45
  twistMotion=Free
 }
 {bone1Name=Dyn_Bags_Bk_R, bone2Name=Bace_Bags_Bk_R, swing1LimitDegrees=5, swing2LimitDegrees=45, twistLimitDegrees=0}
 {
  bone1Name=Bace_Bags_Fr_L
  bone2Name=pelvis
  swing1LimitDegrees=45
  swing1Motion=Free
  swing2LimitDegrees=45
  swing2Motion=Free
  twistLimitDegrees=45
  twistMotion=Free
 }
 {bone1Name=Dyn_Bags_Fr_L, bone2Name=Bace_Bags_Fr_L, swing1LimitDegrees=5, swing2LimitDegrees=45, twistLimitDegrees=0}
 {
  bone1Name=Bace_Bags_Fr_R
  bone2Name=pelvis
  swing1LimitDegrees=45
  swing1Motion=Free
  swing2LimitDegrees=45
  swing2Motion=Free
  twistLimitDegrees=45
  twistMotion=Free
 }
 {bone1Name=Dyn_Bags_Fr_R, bone2Name=Bace_Bags_Fr_R, swing1LimitDegrees=5, swing2LimitDegrees=45, twistLimitDegrees=0}
 {
  bone1Name=Bace_Bags_Mid_L
  bone2Name=pelvis
  swing1LimitDegrees=45
  swing1Motion=Free
  swing2LimitDegrees=45
  swing2Motion=Free
  twistLimitDegrees=45
  twistMotion=Free
 }
 {bone1Name=Dyn_Bags_Mid_L, bone2Name=Bace_Bags_Mid_L, swing1LimitDegrees=5, swing2LimitDegrees=45, twistLimitDegrees=0}
 {
  bone1Name=Bace_Bags_Mid_R
  bone2Name=pelvis
  swing1LimitDegrees=45
  swing1Motion=Free
  swing2LimitDegrees=45
  swing2Motion=Free
  twistLimitDegrees=45
  twistMotion=Free
 }
 {bone1Name=Dyn_Bags_Mid_R, bone2Name=Bace_Bags_Mid_R, swing1LimitDegrees=5, swing2LimitDegrees=45, twistLimitDegrees=0}
]
