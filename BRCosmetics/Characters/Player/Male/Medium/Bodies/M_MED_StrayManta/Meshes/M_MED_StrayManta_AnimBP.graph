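# Post-process style Anim Blueprint for the StrayManta medium male body cosmetic, copying the parent mesh's pose and adding secondary motion.
# The copied pose runs through a Control Rig, pelvis-space RigidBody, and spring-based AnimDynamics on dagger, jacket and hood bones.
# Additive bone transforms then adjust hood and backpack attachment.
# On customization load, part names set back bling flags, which the AnimGraph never reads.
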AnimBlueprint /BRCosmetics/Characters/Player/Male/Medium/Bodies/M_MED_StrayManta/Meshes/M_MED_StrayManta_AnimBP
target_skeleton=&/M_MED_StrayManta_Skeleton
graphs [
 AnimGraph {nodes [
   n1 = Control Rig!(Source=Copy Pose From Mesh() {bUseAttachedParent=true}, Alpha, Rban:bool=Get bRigidBodyEnabled()) -> (Pose) {controlRigAssetReference {blueprintRigClass=&/M_MED_StrayManta_CtrlRig.M_MED_StrayManta_CtrlRig_C}}
   n2 = RigidBody(ComponentPose=Local To Component(LocalPose=@n1)) {
    bClampLinearTranslationLimitToRefPose=true
    bOverrideWorldGravity=true
    overridePhysicsAsset=&/M_MED_StrayManta_Physics
    simulationSpace=BaseBoneSpace
    baseBoneRef {boneName=pelvis}
    overrideWorldGravity {z=-980}
    simSpaceSettings {
     dampingAlpha=0
     maxAngularAcceleration=50
     maxAngularVelocity=3
     maxLinearAcceleration=400
     maxLinearVelocity=250
     worldAlpha=1
     externalLinearDragV=<0.5 0.5 0.5>
    }
   }
   n3 = Anim Dynamics - Bone: Dyn_dager(ComponentPose=@n2, Alpha=1, ExternalForce="0,0,0") {
    angularDampingOverride=1
    angularSpringConstant=700
    bAngularSpring=true
    bOverrideAngularDamping=true
    bUseGravityOverride=true
    simulationSpace=RootRelative
    boundBone {boneName=Dyn_dager}
    componentAppliedLinearAccClamp=<1000 1000 1000>
    componentLinearAccScale=<0.5 0.5 0.5>
    componentLinearVelScale=<0.5 0.5 0.5>
    gravityOverride {z=-980}
    physicsBodyDefinitions [{
      boundBone {boneName=Dyn_dager}
      boxExtents=<10 10 10>
      constraintSetup {angularLimitsMax=<0 0 8>, angularLimitsMin=<0 0 -8>, angularTarget=<1 0 0>}
      localJointOffset=<0.7 0 0>
     }]
   }
   n4 = Anim Dynamics - Bone: Dyn_Jacket_L(ComponentPose=@n3, Alpha=1, ExternalForce="0,0,0") {
    angularDampingOverride=0.999
    angularSpringConstant=700
    bAngularSpring=true
    bOverrideAngularDamping=true
    bUseGravityOverride=true
    simulationSpace=RootRelative
    boundBone {boneName=Dyn_Jacket_L}
    componentAppliedLinearAccClamp=<1000 1000 1000>
    componentLinearAccScale=<0.7 0.7 0.7>
    componentLinearVelScale=<0.7 0.7 0.7>
    gravityOverride {z=-980}
    physicsBodyDefinitions [{
      boundBone {boneName=Dyn_Jacket_L}
      boxExtents=<10 10 10>
      constraintSetup {angularLimitsMax=<0 0 5>, angularLimitsMin=<0 -10 -5>, angularTarget=<1 0 0>}
      localJointOffset=<1 0 0>
     }]
   }
   n5 = Anim Dynamics - Bone: Dyn_Jacket_R(ComponentPose=@n4, Alpha=1, ExternalForce="0,0,0") {
    angularDampingOverride=0.999
    angularSpringConstant=700
    bAngularSpring=true
    bOverrideAngularDamping=true
    bUseGravityOverride=true
    simulationSpace=RootRelative
    boundBone {boneName=Dyn_Jacket_R}
    componentAppliedLinearAccClamp=<1000 1000 1000>
    componentLinearAccScale=<0.7 0.7 0.7>
    componentLinearVelScale=<0.7 0.7 0.7>
    gravityOverride {z=-980}
    physicsBodyDefinitions [{
      boundBone {boneName=Dyn_Jacket_R}
      boxExtents=<10 10 10>
      constraintSetup {angularLimitsMax=<0 0 5>, angularLimitsMin=<0 -10 -5>, angularTarget=<1 0 0>}
      localJointOffset=<1 0 0>
     }]
   }
   n6 = Anim Dynamics - Bone: hood_correct(ComponentPose=@n5, Alpha=1, ExternalForce="0,0,0") {
    bLinearSpring=true
    bOverrideLinearDamping=true
    bUseGravityOverride=true
    linearDampingOverride=1
    linearSpringConstant=100
    simulationSpace=Actor
    boundBone {boneName=hood_correct}
    componentAppliedLinearAccClamp=<100 100 100>
    gravityOverride {z=-980}
    physicsBodyDefinitions [{
      boundBone {boneName=hood_correct}
      boxExtents=<11 11 11>
      constraintSetup {angularLimitsMax=<0 0 0>, angularLimitsMin=<0 0 0>, angularTarget=<0 0 0>, linearAxesMax=<3 3 3>, linearAxesMin=<-3 -3 0>}
      localJointOffset=<0 0 0>
     }]
   }
   n7 = "Transform (Modify) Bone - Bone: hood_correct"!(ComponentPose=@n6, bAlphaBoolEnabled:bool=True, Translation="0,0,0", Rotation="-20,0,0", Scale="1,1,1") -> (Pose) {
    alphaInputType=Bool
    rotationMode=BMM_Additive
    rotationSpace=BCS_BoneSpace
    translationMode=BMM_Additive
    translationSpace=BCS_BoneSpace
    boneToModify {boneName=hood_correct}
    rotation {pitch=-20}
   }
   n8 = "Transform (Modify) Bone - Bone: attach_backpack"!(ComponentPose=@n7, bAlphaBoolEnabled:bool=True, Translation="0,-2,2", Rotation="0,0,-10", Scale="1,1,1") -> (Pose) {
    alphaInputType=Bool
    rotationMode=BMM_Additive
    rotationSpace=BCS_BoneSpace
    translationMode=BMM_Additive
    translationSpace=BCS_BoneSpace
    boneToModify {boneName=attach_backpack}
    rotation {roll=-10}
    translation {y=-2, z=2}
   }
   n9 = "Transform (Modify) Bone - Bone: attach_backpack"(ComponentPose=@n8, bAlphaBoolEnabled=True, Translation="0,2,-3.5", Rotation="0,0,0", Scale="1,1,1") {
    alphaInputType=Bool
    rotationMode=BMM_Additive
    rotationSpace=BCS_BoneSpace
    translationMode=BMM_Additive
    translationSpace=BCS_BoneSpace
    boneToModify {boneName=attach_backpack}
    translation {y=2, z=-3.5}
   }
   Output Pose(Result=Component To Local(ComponentPose=@n9)) {layerGroup=None}
  ]}
 EventGraph {nodes [
   n1 = Event OnCharacterCustomizationLoaded()
   > n2 = For Each Loop(Array:Array of Custom Character Part Object References=@n1.CustomParts) -> (Array Element:Custom Character Part Object Reference)
   > Branch(Condition="Equal (Enum)"(
     A:EFortCustomPartType Enum=GetPartType!(self:Custom Character Part Object Reference=@n2) -> (ReturnValue:EFortCustomPartType Enum=Head)
     B:EFortCustomPartType Enum=Backpack
    ))
   n3 = GetObjectName(Object=@n2)
   > n4 = Branch(Condition=Contains(SearchIn=@n3, Substring=StrayManta))
   > Set IsThisMyBackBling(IsThisMyBackBling:bool=true) -> (Output_Get:bool)
   Set IsThisMyBackBling(execute=@n4.else, IsThisMyBackBling=false)
   > n5 = Branch(Condition=Contains(SearchIn=@n3, Substring=NoBackpack))
   > Set IsBackBlingEquipped(IsBackBlingEquipped:bool=false) -> (Output_Get:bool)
   Set IsBackBlingEquipped(execute=@n5.else, IsBackBlingEquipped=true)
   Event BlueprintUpdateAnimation()
  ]}
]
variables [
 IsBackBlingEquipped {type=boolean}
 IsThisMyBackBling {type=boolean}
]
